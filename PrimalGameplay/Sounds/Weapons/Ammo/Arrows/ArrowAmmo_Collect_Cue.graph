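# Sound cue for collecting arrow ammo pickups, routed through the loot-collect sound class with prop-collect attenuation.
# A Random node picks evenly between two pickup wave variants, then a Modulator applies a slight upward pitch variation.
# A concurrency override allows only two simultaneous instances and stops the oldest when exceeded, at a reduced volume.
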SoundCue /PrimalGameplay/Sounds/Weapons/Ammo/Arrows/ArrowAmmo_Collect_Cue
root=Modulator_0
nodes [
 {id=Modulator_0, children [Random_0], properties {pitchMax=1.1, pitchMin=1}}
 {id=Random_0, children [WavePlayer_0,WavePlayer_1], properties {weights [1,1]}}
 {id=WavePlayer_0, properties {soundWaveAssetPtr=&/ArrowAmmo_Pickup_01}}
 {id=WavePlayer_1, properties {soundWaveAssetPtr=&/ArrowAmmo_Pickup_02}}
]
properties {
 attenuationSettings=/Game/Sounds/Attenuation/Fort_Prop_Collect_Att
 bOverrideConcurrency=true
 maxDistance=2097152
 soundClassObject=&Fort_Loot_Collect
 volumeMultiplier=0.376
 concurrencyOverrides {maxCount=2, resolutionRule=StopOldest}
}
