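# Sound cue for arrow ammo landing, assigned to the Fort_Loot_Collect sound class.
# A Random node picks between two equally weighted landing waves and feeds a Modulator that slightly raises pitch by a varying amount.
# Playback is at reduced volume, with overridden short-range attenuation and a small concurrency cap that prevents new instances once reached.
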SoundCue /PrimalGameplay/Sounds/Weapons/Ammo/Arrows/ArrowAmmo_Landed_Cue
root=Modulator_0
nodes [
 {id=Modulator_0, children [Random_2], properties {pitchMax=1.15, pitchMin=1.05}}
 {id=Random_2, children [WavePlayer_3,WavePlayer_4], properties {weights [1,1]}}
 {id=WavePlayer_3, properties {soundWaveAssetPtr=&/ArrowAmmo_Landed_01}}
 {id=WavePlayer_4, properties {soundWaveAssetPtr=&/ArrowAmmo_Landed_02}}
]
properties {
 bOverrideAttenuation=true
 bOverrideConcurrency=true
 maxDistance=1000
 soundClassObject=&Fort_Loot_Collect
 volumeMultiplier=0.5
 attenuationOverrides {falloffDistance=1000, attenuationShapeExtents {x=0}}
 concurrencyOverrides {maxCount=3, resolutionRule=PreventNew}
}
